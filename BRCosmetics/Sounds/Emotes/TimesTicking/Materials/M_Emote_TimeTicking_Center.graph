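# Translucent unlit material for Niagara mesh particles, forming the center of the TimeTicking emote effect with a clock texture.
# Two tint colors blend by a UV1 radial mask, multiply with particle color and the clock plus edge-weighted Fresnel glow, then feed MF_FX_EmissiveAdjustments.
# Opacity combines particle alpha, softened Fresnel and the inverted radial mask for a fainter center, with DepthFade softening intersections.
# A Desaturation node on the clock texture is unused.
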Material /BRCosmetics/Sounds/Emotes/TimesTicking/Materials/M_Emote_TimeTicking_Center
graph {
 properties {
  bAutomaticallySetUsageInEditor=false
  bUsedWithNiagaraMeshParticles=true
  blendMode=BLEND_Translucent
  shadingModel=MSM_Unlit
  usedShadingModels=Unlit
 }
 nodes [
  n1 = ParticleColor()
  n2 = LinearInterpolate(B=Fresnel() {baseReflectFraction=0, exponent=1}) {constA=1}
  n3 = OneMinus(None=SphereMask(A=TextureCoordinate() {coordinateIndex=1}, B=Constant() {r=0.5}) {attenuationRadius=0.5, hardnessPercent=50})
  DepthFade(
   Opacity=Multiply(A=Multiply(A=@n1.A, B=@n2), B=LinearInterpolate(Alpha=@n3) {constA=0.3})
   FadeDistance=ScalarParameter() {defaultValue=200, parameterName=FadeDistance}
  ) {fadeDistanceDefault=200}
  n4 = TextureSample() {texture=&/T_Emote_TimeTicking_Clock}
  Desaturation(None=@n4.RGB)
  n5 = LinearInterpolate(
   A=VectorParameter() {defaultValue=<rgba 10 2.7507 0.7653 1>}.RGB
   B=VectorParameter() {parameterName=Param_1, defaultValue=<rgba 0.4413 1 0.6306 1>}.RGB
   Alpha=SphereMask(A=TextureCoordinate() {coordinateIndex=1}, B=Constant() {r=0.5}) {attenuationRadius=0.5, hardnessPercent=50}
  )
  n6 = TextureCoordinate() {coordinateIndex=1}
  n7 = Constant() {r=0.5}
  n8 = Multiply(
   A=Multiply(A=Fresnel() {baseReflectFraction=0}) {constB=5}
   B=OneMinus(None=SphereMask(A=@n6, B=@n7) {attenuationRadius=0.5, hardnessPercent=50})
  )
  MaterialFunctionCall(Emissive=Multiply(A=Multiply(A=@n5, B=@n1.RGB) {constA=1}, B=Add(A=@n8, B=@n4.RGB))) {materialFunction=&MF_FX_EmissiveAdjustments}
 ]
}
