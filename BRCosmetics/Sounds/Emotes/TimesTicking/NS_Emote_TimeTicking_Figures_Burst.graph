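# Niagara burst effect for the TimeTicking emote that pops a single number mesh for half a second.
# The active CountNumber emitter takes its orientation and a mesh integer from User.MeshOrientation and User.MeshInt, and applies a system camera offset.
# Curves swell then shrink the mesh while its color shifts from a bright flash to orange.
# Count, FigurePop001 and FigureSprite, including the four-position figure bursts, are disabled, and Glow001 has no renderer.
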NiagaraSystem /BRCosmetics/Sounds/Emotes/TimesTicking/NS_Emote_TimeTicking_Figures_Burst
dependencies {
 usedDataInterfaces [
  &NiagaraDataInterfaceArrayFloat3
  &NiagaraDataInterfaceMeshRendererInfo
  &NiagaraDataInterfaceCurve
  &NiagaraDataInterfaceColorCurve
 ]
 usedDynamicInputs [&FN_FrontEndFloatModifier]
 usedModules [
  &self:SystemSpawnScript.SystemScriptSource.SystemScriptGraph.NiagaraNodeAssignment_0.SetVariables_#1
  &self:SystemSpawnScript.SystemScriptSource.SystemScriptGraph.NiagaraNodeAssignment_1.SetVariables_#2
  &self:SystemSpawnScript.SystemScriptSource.SystemScriptGraph.NiagaraNodeAssignment_3.SetVariables_#3
  &SystemState
 ]
}
emitters {
 Count {bEnabled=false, rendererClasses [], renderers []}
 CountNumber {
  emitterUpdateScript [
   &EmitterState(Life Cycle Mode=::NewEnumerator1, Loop Behavior=::NewEnumerator1, Loop Duration=1, UseLoopDelay=false)
   &SpawnBurst_Instantaneous(Use Spawn Probability=false, UseLoopCountLimit=false)
  ]
  particleSpawnScript [
   &InitializeParticle(
    Color Mode=::NewEnumerator0
    Lifetime=0.5
    Mesh Scale Mode=::NewEnumerator3
    Mesh Uniform Scale=30
    UsePositionOffset=false
    Write Material Random=true
   )
   &InitialMeshOrientation(Mesh Orientation Mode=::NewEnumerator2, Rotation=$User.MeshOrientation, Use Rotation Vector=true)
   &CameraOffset(Camera Offset Amount=$System.CameraOffset)
   /SetVariables(Particles.Meshint32:NiagaraInt32=$User.MeshInt) {isSetParameters=true}
  ]
  particleUpdateScript [
   &ParticleState()
   &ScaleMeshSize(Scale Factor=&Multiply_VectorByFloat(
     Vector=<1 1 1>
     Float=&FloatFromCurve(FloatCurve=DataInterface() {propertyValues {Curve {keys [<at 0 0>,<at 0.2 1.5>,<at 1 0>]}}}, Scale Curve=0.6)
    ))
   &ScaleColor(
    Scale Alpha=&FloatFromCurve(FloatCurve=DataInterface() {propertyValues {Curve {keys [<at 0 2>,<at 1 0.834>]}}})
    Scale RGB=&MakeVectorFromLinearColorRGB(Color=&ColorFromCurve(ColorCurve=DataInterface() {propertyValues {
        AlphaCurve {keys [<at 0 1 RCIM_Cubic>,<at 1 0 RCIM_Cubic>]}
        BlueCurve {keys [<at 0 4.8092 RCIM_Cubic>,<at 0.2771 0>,<at 0.5137 0>,<at 0.9639 0.0878>]}
        GreenCurve {keys [<at 0 10 RCIM_Cubic>,<at 0.2771 0.0525>,<at 0.5137 0.1116>,<at 0.9639 0.5821>]}
        RedCurve {keys [<at 0 8.3331 RCIM_Cubic>,<at 0.2771 0.5>,<at 0.5137 1>,<at 0.9639 5>]}
       }}))
    ScaleA=true
    ScaleRGB=true
   )
  ]
  rendererClasses [&NiagaraMeshRendererProperties]
  renderers [{rendererClass=&NiagaraMeshRendererProperties, rendererIndex=0}]
 }
 FigurePop001 {
  bEnabled=false
  emitterUpdateScript [
   &EmitterState(
    Inactive Response=::NewEnumerator1
    Life Cycle Mode=::NewEnumerator1
    Loop Behavior=::NewEnumerator1
    Loop Delay=0.05
    Loop Duration Mode=::NewEnumerator1
    UseLoopDelay=false
   )
   &SpawnBurst_Instantaneous(Spawn Count=4, Use Spawn Probability=false, UseLoopCountLimit=false) {module=SpawnBurst_Instantaneous003}
   &SpawnBurst_Instantaneous(Spawn Group=1, Spawn Time=0.25, Use Spawn Probability=false, UseLoopCountLimit=false) {enabled=false, module=SpawnBurst_Instantaneous002}
   &SpawnBurst_Instantaneous(Spawn Group=2, Spawn Time=0.5, Use Spawn Probability=false, UseLoopCountLimit=false) {enabled=false, module=SpawnBurst_Instantaneous001}
   &SpawnBurst_Instantaneous(Spawn Group=3, Spawn Time=0.75, Use Spawn Probability=false, UseLoopCountLimit=false) {enabled=false}
  ]
  particleSpawnScript [
   &FN_SetSpawnGroup()
   &InitializeParticle(
    Lifetime=0.56
    Mesh Renderer Array Visibility Mode=::NewEnumerator1
    Mesh Renderer Info=DataInterface() {propertyValues {MeshRenderer=&self:FigurePop001_0.NiagaraMeshRendererProperties_0}}
    Mesh Renderer Mesh Index:NiagaraInt32=$Particles.SpawnGroup
    Mesh Scale=&Multiply_VectorByFloat(Vector=<1 1 1>, Float=30)
    Mesh Scale Mode=::NewEnumerator4
    Position Offset=&AddVectorToPosition(
     Position=$System.FXOffset
     Vector=&SelectVectorFromArray(
      Direct Array Index=$Particles.SpawnGroup
      Vector Selection Array=DataInterface() {propertyValues {FloatData [<0 -66 3>, <0 0 -66>, <0 66 3>, <0 0 66>]}}
     )
    )
    UsePositionOffset=true
    Write Material Random=true
   )
   &InitialMeshOrientation(Mesh Orientation Mode=::NewEnumerator2, Rotation=<0 0 -0.25>, Use Rotation Vector=true)
   &InitialMeshOrientation(
    Mesh Orientation Mode=::NewEnumerator2
    Rotation=&SelectVectorFromArray(
     Direct Array Index=$Particles.SpawnGroup
     Vector Selection Array=DataInterface() {propertyValues {FloatData [<0.25 0 0>, <-0.5 0 0>, <0.25 0 0>, <0 0 0>]}}
    )
    Use Rotation Vector=true
   ) {module=InitialMeshOrientation001}
   &CameraOffset(Camera Offset Amount=$System.CameraOffset) {module=CameraOffset001}
   /SetVariables(Particles.2MO:Quat4f=$Particles.MeshOrientation) {isSetParameters=true}
   /SetVariables() {isSetParameters=true}
  ]
  particleUpdateScript [
   &ParticleState(Kill Particles When Lifetime Has Elapsed=false)
   &UpdateMeshOrientation(Delta Time:NiagaraFloat=$Engine.WorldDeltaTime, Rotation Rate=2, Rotation Vector=<0 0 1>)
   &self:ScratchModule() {module=ScratchModule}
   &ScaleMeshSize(Scale Factor=&Multiply_VectorByFloat(
     Vector=<1 1 1>
     Float=&FloatFromCurve(
      FloatCurve=DataInterface() {propertyValues {Curve {keys [<at 0.1 0>,<at 0.2 1>,<at 0.5 0>]}}}
      CurveIndex=&Divide_Float(
       A=&Subtract_Float(
        A=$Particles.Angle
        B=&MergeHelper_FloatSwitch(0=0, 1=3.14, 2=6.28, 3=9.42, Index=$Particles.SpawnGroup, Num=::NewEnumerator2)
       )
       B=12.56
      )
     )
    )) {module=ScaleMeshSize001}
   &ScaleMeshSize(Scale Factor=&Multiply_VectorByFloat(
     Vector=<1 1 1>
     Float=&FloatFromCurve(FloatCurve=DataInterface() {propertyValues {Curve {keys [<at 0 0>,<at 0.2 1.5>,<at 0.4 1>,<at 1 0>]}}}, Scale Curve=0.6)
    )) {enabled=false}
   &ScaleColor(
    Scale RGB=&MakeVectorFromLinearColorRGB(Color=&ColorFromCurve(ColorCurve=DataInterface() {propertyValues {
        AlphaCurve {keys [<at 0 1 RCIM_Cubic>,<at 1 0 RCIM_Cubic>]}
        BlueCurve {keys [<at 0.9639 0.0878>]}
        GreenCurve {keys [<at 0.9639 0.5821>]}
        RedCurve {keys [<at 0.9639 5>]}
       }}))
    ScaleA=true
    ScaleRGB=true
   )
  ]
  rendererClasses [&NiagaraMeshRendererProperties]
  renderers [{rendererClass=&NiagaraMeshRendererProperties, rendererIndex=0}]
 }
 FigureSprite {
  bEnabled=false
  emitterUpdateScript [
   &EmitterState(
    Delay First Loop Only=true
    Inactive Response=::NewEnumerator1
    Life Cycle Mode=::NewEnumerator1
    Loop Delay=0.05
    Loop Duration=1
    UseLoopDelay=true
   )
   &SpawnBurst_Instantaneous(Spawn Count=5, Use Spawn Probability=false, UseLoopCountLimit=false) {module=SpawnBurst_Instantaneous003}
   &SpawnBurst_Instantaneous(Spawn Count=5, Spawn Group=1, Spawn Time=0.25, Use Spawn Probability=false, UseLoopCountLimit=false) {module=SpawnBurst_Instantaneous002}
   &SpawnBurst_Instantaneous(Spawn Count=5, Spawn Group=2, Spawn Time=0.5, Use Spawn Probability=false, UseLoopCountLimit=false) {module=SpawnBurst_Instantaneous001}
   &SpawnBurst_Instantaneous(Spawn Count=5, Spawn Group=3, Spawn Time=0.75, Use Spawn Probability=false, UseLoopCountLimit=false)
  ]
  particleSpawnScript [
   &FN_SetSpawnGroup()
   &InitializeParticle(
    AdjustAlpha=true
    AdjustHue=true
    AdjustSaturation=true
    AdjustValue=false
    Alpha Scale Range:Vector2f=<0.5 1>
    Color=<rgba 1 0.3797 0.0573 2>
    Color Mode=::NewEnumerator3
    Hue Shift Range:Vector2f=<0 0>
    Lifetime Max=0.7
    Lifetime Min=0.5
    Lifetime Mode=::NewEnumerator1
    Mesh Renderer Array Visibility Mode=::NewEnumerator1
    Mesh Renderer Info=DataInterface() {propertyValues {MeshRenderer=&self:FigureSprite_0.NiagaraMeshRendererProperties_0}}
    Mesh Renderer Mesh Index:NiagaraInt32=$Particles.SpawnGroup
    Mesh Scale=&Multiply_VectorByFloat(Vector=<1 1 1>, Float=10)
    Mesh Scale Mode=::NewEnumerator4
    Position Offset=&AddVectorToPosition(
     Position=$System.FXOffset
     Vector=&SelectVectorFromArray(
      Direct Array Index=$Particles.SpawnGroup
      Vector Selection Array=DataInterface() {propertyValues {FloatData [<0 -90 0>, <0 0 -90>, <0 90 0>, <0 0 90>]}}
     )
    )
    Saturation Range:Vector2f=<1 1>
    UsePositionOffset=true
    Write Material Random=true
   )
   &InitialMeshOrientation(Mesh Orientation Mode=::NewEnumerator2, Rotation=<0 0 -0.25>, Use Rotation Vector=true)
   &ShapeLocation(MaskSpawnGroup=false, Override Seed=false, Sphere Radius=5, UseOwnerScale=true)
   &AddVelocity(
    Module.Use Falloff Distance=false
    Origin Offset=&AddVectorToPosition(
     Position=$System.FXOffset
     Vector=&SelectVectorFromArray(
      Direct Array Index=$Particles.SpawnGroup
      Vector Selection Array=DataInterface() {propertyValues {FloatData [<0 -90 0>, <0 0 -90>, <0 90 0>, <0 0 90>]}}
     )
    )
    Override Seed=false
    Velocity Mode=::NewEnumerator1
    Velocity Speed=50
   )
   &AddVelocity(Velocity=<0 0 300>, Velocity Speed Scale=&RandomRangeFloat(Minimum=0.6, Maximum=1)) {module=AddVelocity001}
   /SetVariables(Particles.SizeScale=1) {isSetParameters=true}
   &SpriteFacingAndAlignment(Sprite Alignment=<0 0 1>, Sprite Facing=<-1 0 0>, WriteAlignment=true, WriteFacing=true)
  ]
  particleUpdateScript [
   &GravityForce(Coordinate Space=::Local, Gravity=<0 -1000 500>)
   &Drag(Drag=8, Use Linear Drag=true, Use Rotational Drag=true)
   &CurlNoiseForce(Bias Noise=false, Coordinate Space:ENiagaraCoordinateSpace=::Local, Noise Frequency=30, Noise Strength=1000, Pan Noise=false, Random Seed=5)
   &GravityForce(Coordinate Space=::Local, Gravity=<0 0 -200>) {module=GravityForce001}
   &SolveForcesAndVelocity(Clamp Velocity=false, Limit Acceleration=false)
   &ParticleState()
   &UpdateMeshOrientation(Rotation Vector=&RandomRangeVector(Minimum=<0 -1 0>, Maximum=<0 1 0>))
   &ScaleMeshSize(Scale Factor=&Multiply_VectorByFloat(
     Vector=&Multiply_VectorByFloat(Vector=<1 1 1>, Float=&RandomRangeFloat(Minimum=1.5, Maximum=1.8))
     Float=&FloatFromCurve(
      FloatCurve=DataInterface() {propertyValues {Curve {keys [<at 0 1 RCIM_Cubic>, {arriveTangent=-0.9142, interpMode=RCIM_Cubic, leaveTangent=-0.9142, tangentMode=RCTM_User, time=1, value=0.5}]}}}
      Scale Curve=&Multiply_Float(A=0.8, B=$Particles.SizeScale)
     )
    ))
   &ScaleColor(
    Scale Alpha=&FloatFromCurve(FloatCurve=DataInterface() {propertyValues {Curve {keys [{arriveTangent=-1.4815, interpMode=RCIM_Cubic, leaveTangent=-1.4815, tangentMode=RCTM_User, time=0, value=1}, <at 1 0>]}}})
    Scale RGB=&Multiply_VectorByFloat(
     Vector=&MakeVectorFromLinearColorRGB(Color=&ColorFromCurve(ColorCurve=DataInterface() {propertyValues {
         AlphaCurve {keys [<at 0 1 RCIM_Cubic>,<at 1 0 RCIM_Cubic>]}
         BlueCurve {keys [
           <at 0 4.8092 RCIM_Cubic>
           {arriveTangent=-9.3621, interpMode=RCIM_Cubic, leaveTangent=-9.3621, time=0.344, value=0.0603}
           <at 0.5137 0>
           <at 0.9639 0.0878>
          ]}
         GreenCurve {keys [
           <at 0 10 RCIM_Cubic>
           {arriveTangent=-19.2497, interpMode=RCIM_Cubic, leaveTangent=-19.2497, time=0.344, value=0.2029}
           <at 0.5137 0.1116>
           <at 0.9639 0.5821>
          ]}
         RedCurve {keys [
           <at 0 8.3331 RCIM_Cubic>
           {arriveTangent=-14.2754, interpMode=RCIM_Cubic, leaveTangent=-14.2754, time=0.344, value=1}
           <at 0.5137 1>
           <at 0.9639 5>
          ]}
        }}))
     Float=&FloatFromCurve(FloatCurve=DataInterface() {propertyValues {Curve {keys [<at 0 0.3>,<at 0.5 0>]}}})
    )
    ScaleA=true
    ScaleRGB=true
   )
   &FN_Sprite_Attributes(
    Color=&MakeLinearColorFromVectorAndFloat(
     "Vector (RGB)"=&MakeVectorFromLinearColorRGB(Color=<rgba 1 0.5579 0.0865 1>)
     "Float (Alpha)"=&Multiply_Float(
      A=&FloatFromCurve(FloatCurve=DataInterface() {propertyValues {Curve {keys [{arriveTangent=-0.4137, interpMode=RCIM_Cubic, leaveTangent=-0.4137, tangentMode=RCTM_User, time=0, value=1}, <at 0.75 0>]}}})
      B=1
     )
    )
    Create Color Attribute=true
    Create SpriteSize Attribute=true
    SpriteSize=&Multiply_Vector2D_ByFloat(Float=$Particles.SizeScale, Vector2D=<200 200>)
    Write to Module Namespace=true
   )
  ]
  rendererClasses [&NiagaraMeshRendererProperties, &NiagaraSpriteRendererProperties]
  renderers [
   {rendererClass=&NiagaraMeshRendererProperties, rendererIndex=0}
   {rendererClass=&NiagaraSpriteRendererProperties, rendererIndex=1}
  ]
 }
 Glow001 {rendererClasses [], renderers []}
}
summary {
 systemName=NS_Emote_TimeTicking_Figures_Burst
 emitters [
  {
   bEnabled=false
   emitterName=FigureSprite
   rendererClasses [&NiagaraMeshRendererProperties, &NiagaraSpriteRendererProperties]
  }
  {bEnabled=false, emitterName=FigurePop001, rendererClasses [&NiagaraMeshRendererProperties]}
  {bEnabled=false, emitterName=Count, rendererClasses []}
  {emitterName=Glow001, rendererClasses []}
  {emitterName=CountNumber, rendererClasses [&NiagaraMeshRendererProperties]}
 ]
 userVariables [
  User.bIsFrontEnd {type=NiagaraBool}
  User.bIsFrontEndPreview {type=NiagaraBool}
  User.DynamicFloat {type=NiagaraFloat, defaultValue=0, description=""}
  User.FrontEndSize {type=NiagaraFloat, defaultValue=0.65, description=""}
  User.MeshInt {type=NiagaraInt32}
  User.MeshOrientation {type=Vector3f, description="", defaultValue {struct=Vector3f, value=<0 0 0>}}
 ]
}
